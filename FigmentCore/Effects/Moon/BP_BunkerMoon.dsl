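; BP_BunkerMoon is a bunker moon effect actor whose garage door light beams can switch off when an actor overlaps it.
; On begin overlap, the authority checks the LightsTurnOffOnOverlap flag, disables collision on the actor and its children, then calls DisableEffect.
; DisableEffect hides the Factory_Garage_Door2_LightBeams component in game.
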(event Collision|EventActorBeginOverlap
  (params (OtherActor :type Actor%20Object%20Reference))
  (Networking|SwitchHasAuthority
    (:Authority
      (if (Variables|Default|GetLightsTurnOffOnOverlap)
        (Collision|SetActorEnableCollisionwithChildren 0)
        (CallFunction|DisableEffect)))))

(event Custom|DisableEffect
  (Development|SetHiddeninGame (Variables|Default|GetFactory_Garage_Door2_LightBeams) true false))

(fn ConstructionScript)
