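; Moon manager that tints the sky dome moon and bunker light according to a watched calendar event's progress.
; On BeginPlay, cosmetic-capable instances await the time-of-day manager, cache the DSA sky dome, then await calendar sync.
; If active, after a short delay it evaluates once, mapping the timespan ratio to red, purple or a lerp, with bunker beams and light using red, purple or fuchsia.
; Otherwise it hides the bunker and applies the starting moon color.
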(event EventBeginPlay
  (bind _currentcolor (Variables|Default|GetCurrentColor))
  ;[0 0.07145 1] Begin Play (Async for TODM)
  (Networking|CanExecuteCosmeticEvents
    (:True
      (Class|FortAsyncActionTimeOfDayManagerReady|GetTimeOfDayManagerReady
        (:OnReady
          (CallFunction|GetDSA)
          (MCPContext|InitialCalendarSyncCompleted
            (:OnSyncComplete
              ; Watches Calendar and Updates Lerp Color
              (bind (_biseventactive _timesincebegin _timeuntilend _timespanratio) (CalendarEvents|GetCalendarEventInformation (Variables|Default|GetEvent%20to%20Watch)))
              (if _biseventactive
                (Variables|Default|SetChangeRate _timespanratio)
                (Utilities|FlowControl|Delay 1.0)
                ; Set Bunker and Moon Color
                (CallFunction|SetLerpColor)
                (CallFunction|SetBunkerColors _currentcolor)
                (CallFunction|SetMoonMatColor _currentcolor)
                (else
                  ; Set Moon to Original Sequence Color and Bunker = Hidden
                  (Rendering|SetActorHiddenInGame (Variables|Default|GetBunkerLight) true)
                  (CallFunction|SetMoonMatColor (Variables|Default|GetMoonColorStart)))))))))))

(fn GetDSA
  (bind _returnvalue (Actor|GetActorOfClass "/FigmentCore/TimeOfDay/TODM/TOD_BP/DSA_Figment_Ch1_Moon.DSA_Figment_Ch1_Moon_C"))
  (Variables|Default|SetSky%20Dome (Variables|/Script/FortniteGame.FortDaySequenceActor|GetSkyDomeComponent _returnvalue))
  (return))

(fn SetBunkerColors
  (params (Color :type Linear%20Color%20Structure))
  (bind _bunkerred (Variables|Default|GetBunkerRed))
  (bind _bunkerfuchsia (Variables|Default|GetBunkerFuchsia))
  (bind _bunkerpurple (Variables|Default|GetBunkerPurple))
  (Rendering|SetActorHiddenInGame (Variables|Default|GetBunkerLight) false)
  (Utilities|FlowControl|Sequence
    (:then_0
      (if (== Color (Variables|Default|GetRed))
        (Rendering|Material|SetColorParameterValueonMaterials (Variables|/FigmentCore/Effects/Moon/BP_BunkerMoon.BP_BunkerMoon_C|GetFactory_Garage_Door2_LightBeams (Variables|Default|GetBunkerLight)) "Color" _bunkerred)
        (Rendering|Components|Light|SetLightColor (Variables|/FigmentCore/Effects/Moon/BP_BunkerMoon.BP_BunkerMoon_C|GetColorfulLight (Variables|Default|GetBunkerLight)) _bunkerred true)))
    (:then_1
      (if (== Color (Variables|Default|GetPurple))
        (Rendering|Material|SetColorParameterValueonMaterials (Variables|/FigmentCore/Effects/Moon/BP_BunkerMoon.BP_BunkerMoon_C|GetFactory_Garage_Door2_LightBeams (Variables|Default|GetBunkerLight)) "Color" _bunkerpurple)
        (Rendering|Components|Light|SetLightColor (Variables|/FigmentCore/Effects/Moon/BP_BunkerMoon.BP_BunkerMoon_C|GetColorfulLight (Variables|Default|GetBunkerLight)) _bunkerpurple true)))
    (:then_2
      (if (and (!= Color (Variables|Default|GetRed)) (!= Color (Variables|Default|GetPurple)))
        (Rendering|Material|SetColorParameterValueonMaterials (Variables|/FigmentCore/Effects/Moon/BP_BunkerMoon.BP_BunkerMoon_C|GetFactory_Garage_Door2_LightBeams (Variables|Default|GetBunkerLight)) "Color" _bunkerfuchsia)
        (Rendering|Components|Light|SetLightColor (Variables|/FigmentCore/Effects/Moon/BP_BunkerMoon.BP_BunkerMoon_C|GetColorfulLight (Variables|Default|GetBunkerLight)) _bunkerfuchsia true))))
  (return))

(fn SetLerpColor
  (Utilities|FlowControl|Sequence
    (:then_0
      (if (<= (Variables|Default|GetChangeRate) (Variables|Default|GetMaxRed))
        (Variables|Default|SetCurrentColor (Variables|Default|GetRed))))
    (:then_1
      (if (>= (Variables|Default|GetChangeRate) (Variables|Default|GetMinPurple))
        (Variables|Default|SetCurrentColor (Variables|Default|GetPurple))))
    (:then_2
      (if (and (>= (Variables|Default|GetChangeRate) (Variables|Default|GetMaxRed)) (<= (Variables|Default|GetChangeRate) (Variables|Default|GetMinPurple)))
        (Variables|Default|SetCurrentColor (Math|Color|Lerp(LinearColor) (Variables|Default|GetRed) (Variables|Default|GetPurple) (Variables|Default|GetChangeRate))))))
  (return))

(fn SetMoonMatColor
  (params (Color :type Linear%20Color%20Structure))
  (Rendering|Material|SetVectorParameterValue (Class|BTechArtGlobalLibrary|GetDynamicMaterialInstance (Variables|Default|GetSky%20Dome)) "MoonColor" Color :!class /Script/Engine.MaterialInstanceDynamic))

(fn ConstructionScript)
